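# PivotPainter material function that, by its name, guards vector components against being zero.
# It splits the input, declared scalar, with BreakOutFloat3Components and runs each channel through an If against a default-zero constant.
# Unequal values pass through unchanged, while equal ones take a second constant, shown as 0 here.
# Results are output as a recombined RGB vector and as separate R, G and B scalars.
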MaterialFunction /Engine/Functions/Engine_MaterialFunctions02/PivotPainter/ProtectFrom0
graph {nodes [
  n1 = MaterialFunctionCall(Float3=FunctionInput() {inputType=FunctionInput_Scalar}) {materialFunction=&BreakOutFloat3Components}
  n2 = Constant() {desc=1.00001 - needs the .00001 at the end}
  n3 = Constant() {r=0}
  n4 = If(A=@n1.R, B=@n2, "A > B"=@n1.R, "A == B"=@n3, "A < B"=@n1.R)
  n5 = If(A=@n1.G, B=@n2, "A > B"=@n1.G, "A == B"=@n3, "A < B"=@n1.G)
  n6 = If(A=@n1.B, B=@n2, "A > B"=@n1.B, "A == B"=@n3, "A < B"=@n1.B)
  FunctionOutput(None=AppendVector(A=AppendVector(A=@n4, B=@n5), B=@n6)) {outputName=RGB}
  FunctionOutput(None=@n4) {outputName=R, sortPriority=1}
  FunctionOutput(None=@n5) {outputName=G, sortPriority=2}
  FunctionOutput(None=@n6) {outputName=B, sortPriority=3}
 ]}
