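# Material function that decodes the hierarchy data Pivot Painter MAXScript bakes into meshes, exposing world-space parent and child pivots and X axes.
# Vertex RGB, scaled by a max-distance input, gives parent pivot; alpha and UV3.G give parent axis, with derived Z and packed sign.
# UV2 and UV3.R give child pivot, their fractional parts child axis.
# It also outputs object pivot, orientation, scale and a black-paint mask.
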MaterialFunction /Engine/Functions/Engine_MaterialFunctions02/PivotPainter/PivotPainter_HierarchyData
graph {
 properties {
  bExposeToLibrary=true
  description="Processes and organizes world position and angle information stored in the model's uvs by the Pivot Painter MAXScript. "
  libraryCategoriesText [Pivot Painter]
 }
 nodes [
  FunctionOutput(None=ComponentMask(None=MaterialFunctionCall() {materialFunction=/Engine/Functions/Engine_MaterialFunctions02/WorldPositionOffset/ObjectScale}.Scale XYZ) {r=true}) {outputName=Object Scale, sortPriority=23}
  n1 = FunctionInput(Preview=Constant() {r=4096}) {
   bUsePreviewValueAsDefault=true
   description="This value should match the value that was used in the max script \"Max Dist for Parent Piv\" spinner while painting your asset. "
   inputName=Max Dist for Parent Piv
   inputType=FunctionInput_Scalar
  }
  n2 = Multiply(A=@n1, B=Constant3Vector() {constant=<rgba 2 2 1 0>})
  n3 = Subtract(A=VertexColor(), B=Constant3Vector() {constant=<rgba 0.5 0.5 0 0>})
  FunctionOutput(None=TransformPosition(None=Multiply(A=@n2, B=@n3) {desc=ParentPivotPoint}) {desc="L > W", transformSourceType=TRANSFORMPOSSOURCE_Instance, transformType=TRANSFORMPOSSOURCE_World}) {
   description="Returns the pivot point information for each parent in world space. Children return the pivot point information of the parent they're attached to. "
   outputName=Parent Piv Position
  }
  n4 = VertexColor()
  n5 = ComponentMask(None=TextureCoordinate() {coordinateIndex=3}) {g=true}
  n6 = If(A=@n5, B=Constant() {desc=1.00001 - needs the .00001 at the end, r=1}, "A > B"=Constant() {r=1}, "A < B"=Constant())
  n7 = ConstantBiasScale(None=Add(A=@n5, B=LinearInterpolate(Alpha=@n6) {constAlpha=1, constB=-1})) {bias=-0.5, desc=XY, scale=2}
  n8 = LinearInterpolate(Alpha=@n6) {constA=1, constAlpha=1, constB=-1}
  n9 = Multiply(A=@n8) {constB=-1}
  n10 = MaterialFunctionCall!(In=AppendVector(A=ConstantBiasScale(None=@n4.A) {bias=-0.5, desc=XX, scale=2}, B=Multiply(A=@n7, B=@n9)) {desc=XY}) {materialFunction=&/ProtectFrom0}
  n11 = SquareRoot(None=MaterialFunctionCall(f2=Multiply(A=@n10.RGB, B=@n10.RGB)) {materialFunction=&AddComponents}.f2)
  n12 = Divide(B=Add(A=@n11) {constB=0, desc=0.000001})
  n13 = DeriveNormalZ(InXY=If(A=@n11, B=Constant() {desc=0.999, r=0.999}, "A > B"=Multiply(A=@n10.RGB, B=@n12), "A < B"=@n10.RGB))
  n14 = AppendVector(A=ComponentMask(None=@n13) {g=true, r=true}, B=Multiply(A=ComponentMask(None=@n13) {b=true}, B=@n8))
  FunctionOutput(None=MaterialFunctionCall(Vector=Transform(None=@n14) {desc="L -> W", transformSourceType=TRANSFORMSOURCE_Local}, Default=Constant3Vector() {constant=<rgba 0 0 1 0>}) {materialFunction=&SafeNormalize}.Result) {description="Returns a normalized vector pointing down the parent's x-axis. ", outputName=Parent X Axis Vector, sortPriority=2}
  n15 = VertexColor()
  n16 = Add(A=Add(A=@n15.R, B=@n15.R), B=@n15.B)
  n17 = Constant()
  n18 = Constant() {r=1}
  FunctionOutput(None=If(A=Add(A=@n16, B=@n15.A), B=@n17, "A > B"=@n18, "A == B"=@n17, "A < B"=@n18)) {
   description=Returns a black value for surfaces that are painted black by the pivot painter tool. This mask only works in the vertex shader.
   outputName=Black Mask
   sortPriority=40
  }
  FunctionOutput(None=TransformPosition(None=Constant3Vector()) {desc="L > W", transformSourceType=TRANSFORMPOSSOURCE_Instance, transformType=TRANSFORMPOSSOURCE_World}) {outputName="Object  Pivot Point", sortPriority=20}
  FunctionOutput() {description=Empty, outputName=------------------------, sortPriority=9}
  n19 = AppendVector(A=TextureCoordinate() {coordinateIndex=2}, B=ComponentMask(None=TextureCoordinate() {coordinateIndex=3}) {r=true})
  n20 = MaterialFunctionCall(Vector=ConstantBiasScale(None=Frac(None=@n19)) {bias=-0.5, scale=2}, Default=Constant3Vector() {constant=<rgba 0 0 1 0>}) {materialFunction=&SafeNormalize}
  FunctionOutput(None=Normalize(VectorInput=Transform(None=@n20.Result) {transformSourceType=TRANSFORMSOURCE_Instance})) {description=Returns a normalized vector pointing down the parent's x-axis., outputName=Child X-Axis Vector, sortPriority=12}
  n21 = Constant3Vector() {constant=<rgba 0 0 1 0>}
  FunctionOutput(None=Normalize(VectorInput=Transform(None=@n21) {transformSourceType=TRANSFORMSOURCE_Instance})) {outputName=Object Orientation, sortPriority=21}
  FunctionOutput() {description=Empty, outputName="------------------------ ", sortPriority=15}
  FunctionOutput(None=TransformPosition(None=@n19) {desc="L > W", transformSourceType=TRANSFORMPOSSOURCE_Instance, transformType=TRANSFORMPOSSOURCE_World}) {
   description="Return the pivot position information in world space for each child. Parents return a value of (0,0,0)."
   outputName=Child Piv Position
   sortPriority=10
  }
  n22 = Divide(A=@n1) {constB=255, desc=world units per step}
  n23 = Subtract(A=VertexColor(), B=Constant3Vector() {constant=<rgba 0.5 0.5 0 0>})
  n24 = Multiply(A=Constant3Vector() {constant=<rgba -1 -1 1 0>}, B=OneMinus(None=@n23))
  n25 = Multiply(A=@n23, B=Constant3Vector() {constant=<rgba 2 2 1 0>})
  n26 = Multiply(A=@n25) {constB=255}
  Multiply(A=Add(A=Multiply(A=@n22, B=@n24), B=Multiply(A=@n1, B=@n25) {desc=ParentPivotPoint}), B=Clamp(None=Abs(None=@n26)))
 ]
}
