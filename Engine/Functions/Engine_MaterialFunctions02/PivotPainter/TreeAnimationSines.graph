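# Foliage Painter material function that turns a time-plus-gradient scalar into an irregular oscillation for animating trees.
# The input feeds four LinearSine calls in -1 to 1 mode, each with its own adjustable period.
# The product of two rounded waves is added to two clamped linear waves, one keeping only its positive part, to give a single scalar output.
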MaterialFunction /Engine/Functions/Engine_MaterialFunctions02/PivotPainter/TreeAnimationSines
graph {
 properties {description=Add time and a gradient to create an animation used for trees, libraryCategoriesText [Foliage Painter]}
 nodes [
  n1 = ComponentMask(None=FunctionInput(Preview=Time()) {bUsePreviewValueAsDefault=true, description=FoliageSineWaves, inputName=Time and Grad, inputType=FunctionInput_Scalar}) {r=true}
  n2 = FunctionInput() {bUsePreviewValueAsDefault=true, inputName=LS1, inputType=FunctionInput_Scalar, sortPriority=1, previewValue=<8.33 0 0 1>}
  n3 = StaticBool() {value=true}
  n4 = Clamp(None=MaterialFunctionCall(Value=@n1, Period=@n2, -1 to 1=@n3) {materialFunction=&LinearSine}.Linear Sine) {maxDefault=0.5}
  n5 = FunctionInput() {bUsePreviewValueAsDefault=true, inputName=LS2, inputType=FunctionInput_Scalar, sortPriority=2, previewValue=<2 0 0 1>}
  n6 = FunctionInput() {bUsePreviewValueAsDefault=true, inputName=LS3, inputType=FunctionInput_Scalar, sortPriority=3, previewValue=<4.5 0 0 1>}
  n7 = FunctionInput() {bUsePreviewValueAsDefault=true, inputName=LS4, inputType=FunctionInput_Scalar, sortPriority=4, previewValue=<3.154 0 0 1>}
  n8 = Add(
   A=Multiply(
    A=MaterialFunctionCall(Value=@n1, Period=@n5, -1 to 1=@n3) {materialFunction=&LinearSine}."Rounded Linear Sine "
    B=MaterialFunctionCall(Value=@n1, Period=@n6, -1 to 1=@n3) {materialFunction=&LinearSine}."Rounded Linear Sine "
   )
   B=Clamp(None=MaterialFunctionCall(Value=@n1, Period=@n7, -1 to 1=@n3) {materialFunction=&LinearSine}.Linear Sine) {maxDefault=0.5, minDefault=-0.5}
  )
  FunctionOutput(None=ComponentMask(None=Add(A=@n4, B=@n8)) {r=true})
 ]
}
