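# Material function that returns object-level data: world-space pivot, per-axis and uniform scale, and a transformed input vector.
# Instance-to-world transforms of the origin give the pivot, which is subtracted from transformed (1,1,1) for XYZ scale.
# Uniform scale is the distance from pivot to a transformed normalized diagonal, independent of orientation.
# Local Vector, defaulting to +Z, passes through a default-setting Transform node to become World Vector.
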MaterialFunction /Engine/Functions/Engine_MaterialFunctions02/ObjectLocalToWorldData
graph {
 properties {description=Returns object level information., libraryCategoriesText [Misc, ""]}
 nodes [
  Constant3Vector() {constant=<rgba 0.58 0.58 0.58 0>}
  n1 = Constant3Vector() {constant=<rgba 1 1 1 0>}
  n2 = TransformPosition(None=Constant3Vector()) {transformSourceType=TRANSFORMPOSSOURCE_Instance, transformType=TRANSFORMPOSSOURCE_World}
  FunctionOutput(None=Subtract(A=TransformPosition(None=@n1) {transformSourceType=TRANSFORMPOSSOURCE_Instance, transformType=TRANSFORMPOSSOURCE_World}, B=@n2)) {description="Returns a vector value that is equal to the objects scale on the x,y and z axes.", outputName=Object Scale XYZ, sortPriority=2}
  n3 = Normalize(VectorInput=Constant3Vector() {constant=<rgba 1 1 1 0>})
  FunctionOutput(None=Distance(A=TransformPosition(None=@n3) {transformSourceType=TRANSFORMPOSSOURCE_Instance, transformType=TRANSFORMPOSSOURCE_World}, B=@n2) {desc=finds scale regardless of orientation}) {description=Returns a scalar value that is equal to the objects uniform scale., outputName=Object Scale, sortPriority=2}
  n4 = FunctionInput() {
   bUsePreviewValueAsDefault=true
   description="Input a vector 3 value to be transformed by the objects rotation. For example, a value of (1,0,0) will return a world space vector along the objects x-axis regardless of object orientation."
   inputName=Local Vector
   previewValue=<0 0 1 1>
  }
  FunctionOutput(None=Transform(None=@n4)) {description=Local vector input transformed into world space., outputName=World Vector, sortPriority=5}
  FunctionOutput(None=@n2) {description=Object pivot point in world space., outputName=Object Pivot Location}
 ]
}
