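# Pivot Painter material function that decodes per-element pivot, axis and mask data baked into mesh UVs and vertex colors.
# Pivot position combines UV2 and UV3.r, offsets Y by -1, and transforms from instance to world space.
# Vertex RGB unpacks into an X-axis vector transformed from local space; alpha gives custom falloff, UV3.g a per-element random value.
# It also outputs object scale and a black-paint mask summing red twice, blue and alpha, omitting green.
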MaterialFunction /Engine/Functions/Engine_MaterialFunctions02/PivotPainter/PivotPainter_PerObjectData
graph {
 properties {
  bExposeToLibrary=true
  description="Processes and organizes world position and angle information stored in the model's uvs by the Pivot Painter MAXScript. "
  libraryCategoriesText [Pivot Painter]
 }
 nodes [
  FunctionOutput(None=ComponentMask(None=VertexColor().A) {r=true}) {
   description=Returns the custom falloff values stored in the vertex alpha channel of the model.
   outputName=Custom Alpha Values
   sortPriority=14
  }
  FunctionOutput(None=ComponentMask(None=TextureCoordinate() {coordinateIndex=3}) {g=true}) {description=Returns a random value per element in the 0-1 range., outputName=Random Value Per Element, sortPriority=13}
  n1 = Normalize(VectorInput=ConstantBiasScale(None=VertexColor()) {bias=-0.5, scale=2})
  FunctionOutput(None=Normalize(VectorInput=Transform(None=@n1) {transformSourceType=TRANSFORMSOURCE_Local})) {
   description=Returns a normalized vector pointing down the elements X-Axis. Which points from the pivot point toward the averaged center of the mesh.
   outputName=X-Axis Vector
   sortPriority=9
  }
  n2 = AppendVector(A=TextureCoordinate() {coordinateIndex=2}, B=ComponentMask(None=TextureCoordinate() {coordinateIndex=3}) {r=true})
  n3 = Constant3Vector() {constant=<rgba 0 -1 0 0>}
  FunctionOutput(None=TransformPosition(None=Add(A=@n2, B=@n3)) {transformSourceType=TRANSFORMPOSSOURCE_Instance, transformType=TRANSFORMPOSSOURCE_World}) {description=Returns the pivot point information for each element in world space., outputName=Pivot Position, sortPriority=10}
  n4 = VertexColor()
  n5 = Add(A=Add(A=@n4.R, B=@n4.R), B=@n4.B)
  n6 = Constant()
  n7 = Constant() {r=1}
  FunctionOutput(None=If(A=Add(A=@n5, B=@n4.A), B=@n6, "A > B"=@n7, "A == B"=@n6, "A < B"=@n7)) {
   description=Returns a black value for surfaces that are painted black by the pivot painter tool. This mask only works in the vertex shader.
   outputName=Black Mask
   sortPriority=40
  }
  FunctionOutput(None=MaterialFunctionCall() {materialFunction=/Engine/Functions/Engine_MaterialFunctions02/WorldPositionOffset/ObjectScale}.Scale XYZ) {description=Returns a scalar value that is equal to the objects uniform scale., outputName=Object Scale, sortPriority=20}
 ]
}
