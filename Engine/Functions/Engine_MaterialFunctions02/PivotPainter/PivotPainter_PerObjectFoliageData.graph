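# Material function that unpacks Pivot Painter per-element foliage data for wind-driven vertex animation.
# UV2 and UV3.r form pivots transformed from instance to world space; UV3.g gives random values, vertex alpha custom falloff.
# The element X-axis comes from vertex color or, via static switch, an input vector; crossing wind direction with it yields the rotation axis.
# Nested Wind and ObjectLocalToWorldData functions supply wind terms, object scale and pivot.
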MaterialFunction /Engine/Functions/Engine_MaterialFunctions02/PivotPainter/PivotPainter_PerObjectFoliageData
graph {
 properties {
  bExposeToLibrary=true
  description="Processes and organizes world position and angle information stored in the model's uvs by the Pivot Painter MAXScript. "
  libraryCategoriesText [Pivot Painter]
 }
 nodes [
  n1 = FunctionInput() {
   bUsePreviewValueAsDefault=true
   desc="!!!!!!!!!! Needs to be replaced with wind vector"
   description=Needs to be replaced with wind vector.
   inputName=Wind Vector
   inputType=FunctionInput_Vector4
   previewValue=<0 1 0 1>
  }
  n2 = MaterialFunctionCall!(WindActor=@n1) {materialFunction=/Engine/Functions/Engine_MaterialFunctions02/WorldPositionOffset/Wind}
  FunctionOutput(None=@n2.WindSpeed) {description=Wind speed multiplied by time multiplied by -1., outputName=Wind Accumulator, sortPriority=16}
  FunctionOutput(None=ComponentMask(None=VertexColor().A) {r=true}) {
   description="Returns the custom falloff values stored in the vertex alpha channel of the model. "
   outputName=Custom Alpha Values
   sortPriority=14
  }
  FunctionOutput(None=@n2.Wind Strength) {
   description="Returns wind strength. The magnitude of the wind vector found by calculating distance from the wind vector to 0. "
   outputName=Wind Strength
   sortPriority=17
  }
  FunctionOutput(None=ComponentMask(None=@n1) {a=true}) {outputName=Wind Speed, sortPriority=17}
  FunctionOutput(None=ComponentMask(None=TextureCoordinate() {coordinateIndex=3}) {g=true}) {description=Returns a random value per element in the 0-1 range., outputName=Random Value Per Element, sortPriority=13}
  n3 = FunctionInput() {
   bUsePreviewValueAsDefault=true
   description="Enter a local vector to be used as the elements rotation axis. If the optimized for foliage placement boolean is true. "
   inputName=Optimized Vector
   sortPriority=1
   previewValue=<0 0 1 1>
  }
  n4 = FunctionInput(Preview=StaticBool()) {
   bUsePreviewValueAsDefault=true
   description="Default value = false. Check if you processed the mesh using Pivot Painter with the optimize for Foliage Placement option checked."
   inputName=Optimized for Foliage Placement
   inputType=FunctionInput_StaticBool
  }
  n5 = StaticSwitch(True=@n3, False=ConstantBiasScale(None=VertexColor()) {bias=-0.5, scale=2}, Value=@n4)
  n6 = Normalize(VectorInput=Transform(None=Normalize(VectorInput=@n5)) {transformSourceType=TRANSFORMSOURCE_Instance})
  FunctionOutput(None=@n6) {
   description=Returns a normalized vector pointing down the elements X-Axis. Which points from the pivot point toward the averaged center of the mesh.
   outputName=Element X-Axis
   sortPriority=12
  }
  n7 = Normalize(VectorInput=ComponentMask(None=@n1) {b=true, g=true, r=true})
  FunctionOutput(None=Normalize(VectorInput=CrossProduct(A=@n7, B=@n6))) {
   description="Returns a rotation axis for branches to be used with the \"RotateAboutAxis\" node. Leaves return the same information. Note: The angle is found by transforming a vector along the branch's x-axis from local to world space. Then the cross product is found between the wind direction and the transformed vector."
   outputName=Element Rot Axis
   sortPriority=11
  }
  n8 = TextureCoordinate() {coordinateIndex=2}
  n9 = ComponentMask(None=TextureCoordinate() {coordinateIndex=3}) {r=true}
  FunctionOutput(None=TransformPosition(None=AppendVector(A=@n8, B=@n9)) {transformSourceType=TRANSFORMPOSSOURCE_Instance, transformType=TRANSFORMPOSSOURCE_World}) {description=Returns the pivot point information for each element in world space., outputName=Piv Position, sortPriority=10}
  n10 = MaterialFunctionCall!(Local Vector) {materialFunction=/Engine/Functions/Engine_MaterialFunctions02/ObjectLocalToWorldData}
  FunctionOutput(None=@n10.Object Scale XYZ) {description=Returns a scalar value that is equal to the objects uniform scale., outputName=Object Scale XYZ, sortPriority=20}
  FunctionOutput(None=@n10.Object Scale) {description=Returns a scalar value that is equal to the objects uniform scale., outputName=Uniform Object Scale, sortPriority=20}
  FunctionOutput(None=@n2.Normalized Wind Vector) {outputName=Normalized Wind Vector, sortPriority=15}
  FunctionOutput(None=@n10.Object Pivot Location) {description=Returns a scalar value that is equal to the objects uniform scale., outputName=Object PIvot Location, sortPriority=20}
 ]
}
