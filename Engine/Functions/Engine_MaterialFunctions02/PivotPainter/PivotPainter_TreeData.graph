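# Pivot Painter tree material function that decodes painted branch and leaf pivots and axes for wind-driven foliage animation.
# Branch pivots come from vertex color RGB scaled by max distance; axes from alpha and UV3.G, rebuilding Z.
# Leaf pivots come from UV2 and UV3.R, whose fractional parts encode axes.
# Instance-to-world transforms feed rotation axes and flow gradients using WindVector or the Wind function, alongside object-level data and masks.
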MaterialFunction /Engine/Functions/Engine_MaterialFunctions02/PivotPainter/PivotPainter_TreeData
graph {
 properties {
  bExposeToLibrary=true
  description="Processes and organizes world position and angle information stored in the model's uvs by the painter script. The outputs starting with (tree) process the models uv information as it would be stored by the  Pivot Painter MAXScript. The outputs starting with (Leaf) process the uv information as it would be stored by the per object pivot painting section of the script."
  libraryCategoriesText [Pivot Painter]
 }
 nodes [
  FunctionOutput(None=MaterialFunctionCall() {materialFunction=/Engine/Functions/Engine_MaterialFunctions02/WorldPositionOffset/ObjectScale}.Scale Z) {outputName=Object Scale, sortPriority=23}
  n1 = FunctionInput(Preview=Constant() {r=4096}) {
   bUsePreviewValueAsDefault=true
   description="This value should match the value that was used in the max script \"Max Dist for Parent Piv\" spinner while painting your asset. "
   inputName=Max Dist for Parent Piv
   inputType=FunctionInput_Scalar
  }
  n2 = Constant3Vector() {constant=<rgba 2 2 1 0>}
  n3 = VertexColor()
  n4 = Constant3Vector() {constant=<rgba 0.5 0.5 0 0>}
  n5 = TransformPosition(None=Multiply(A=Multiply(A=@n1, B=@n2), B=Subtract(A=@n3, B=@n4)) {desc=ParentPivotPoint}) {desc="L > W", transformSourceType=TRANSFORMPOSSOURCE_Instance, transformType=TRANSFORMPOSSOURCE_World}
  FunctionOutput(None=@n5) {
   description="Returns the pivot point information for each branch in world space. Leaves return the pivot point information of the branch they're attached to. "
   outputName=Branch Piv Position
  }
  FunctionOutput() {description=Empty, outputName="------------------------   ", sortPriority=35}
  n6 = ComponentMask(None=TextureCoordinate() {coordinateIndex=3}) {g=true}
  n7 = If(A=@n6, B=Constant() {desc=1.00001 - needs the .00001 at the end, r=1}, "A > B"=Constant() {r=1}, "A < B"=Constant())
  n8 = Add(A=@n6, B=LinearInterpolate(Alpha=@n7) {constAlpha=1, constB=-1})
  n9 = LinearInterpolate(Alpha=@n7) {constA=1, constAlpha=1, constB=-1}
  n10 = AppendVector(
   A=ConstantBiasScale(None=VertexColor().A) {bias=-0.5, desc=XX, scale=2}
   B=Multiply(A=ConstantBiasScale(None=@n8) {bias=-0.5, desc=XY, scale=2}, B=Multiply(A=@n9) {constB=-1})
  ) {desc=XY}
  n11 = MaterialFunctionCall(f2=Multiply(A=@n10, B=@n10)) {materialFunction=&AddComponents}
  n12 = Divide(B=Add(A=SquareRoot(None=@n11.f2)) {constB=0, desc=0.000001})
  n13 = DeriveNormalZ(InXY=If(A=@n11.f2, B=Constant() {desc=0.999, r=0.999}, "A > B"=Multiply(A=@n10, B=@n12), "A < B"=@n10))
  n14 = ComponentMask(None=@n13) {g=true, r=true}
  n15 = Multiply(A=ComponentMask(None=@n13) {b=true}, B=@n9)
  n16 = MaterialFunctionCall(Vector=Transform(None=AppendVector(A=@n14, B=@n15)) {desc="L -> W", transformSourceType=TRANSFORMSOURCE_Instance}) {materialFunction=&SafeNormalize}
  n17 = Normalize(VectorInput=FunctionInput() {bUsePreviewValueAsDefault=true, inputName=WindVector, previewValue=<0 1 0 1>})
  FunctionOutput(None=Normalize(VectorInput=CrossProduct(A=@n16.Result, B=@n17))) {
   description="Returns a rotation axis for branches to be used with the \"RotateAboutAxis\" node. Leaves return the same information. Note: The angle is found by transforming a vector along the branch's x-axis from local to world space. Then the cross product is found between the wind direction and the transformed vector."
   outputName=Branch Wind Rot Axis
   sortPriority=1
  }
  FunctionOutput(None=@n16.Result) {
   description=Returns a normalized vector pointing down the branches x-axis. This is not typically needed for foliage animations unless it is needed for custom vector calculations.
   outputName=Branch-X Axis Vector
   sortPriority=2
  }
  n18 = AppendVector(A=TextureCoordinate() {coordinateIndex=2}, B=ComponentMask(None=TextureCoordinate() {coordinateIndex=3}) {r=true})
  n19 = ComponentMask(None=TransformPosition(None=@n18) {desc="L > W", transformSourceType=TRANSFORMPOSSOURCE_Instance, transformType=TRANSFORMPOSSOURCE_World}) {b=true, g=true, r=true}
  n20 = MaterialFunctionCall!(WindActor) {materialFunction=/Engine/Functions/Engine_MaterialFunctions02/WorldPositionOffset/Wind}
  FunctionOutput(None=ComponentMask(None=DotProduct(A=@n19, B=@n20.Normalized Wind Vector)) {r=true}) {description=Returns a gradient value in the direction of the blowing wind., outputName=Leaf flow Grad, sortPriority=14}
  FunctionOutput(None=@n20.WindSpeed) {description=Wind speed multiplied by time., outputName=WindSpeed, sortPriority=33}
  n21 = CrossProduct(A=@n20.Normalized Wind Vector, B=Constant3Vector() {constant=<rgba 0 0 1 0>})
  FunctionOutput(None=ComponentMask(None=DotProduct(A=@n19, B=@n21)) {r=true}) {description=Returns a gradient value across world space at a 90 degree angle to the wind., outputName=Leaf flow Grad 90 deg, sortPriority=15}
  FunctionOutput(None=@n20.Wind Strength) {
   description="Returns wind strength. The magnitude of the wind vector found by calculating distance from the wind vector to 0. "
   outputName=WindStrength
   sortPriority=31
  }
  n22 = VertexColor()
  n23 = Add(A=Add(A=@n22.R, B=@n22.R), B=@n22.B)
  n24 = Constant()
  n25 = Constant() {r=1}
  FunctionOutput(None=If(A=Add(A=@n23, B=@n22.A), B=@n24, "A > B"=@n25, "A == B"=@n24, "A < B"=@n25)) {
   description=Returns a black value for surfaces that are painted black by the pivot painter tool. This mask only works in the vertex shader.
   outputName=Black Mask
   sortPriority=40
  }
  n26 = ComponentMask(None=TransformPosition(None=Constant3Vector()) {desc="L > W", transformSourceType=TRANSFORMPOSSOURCE_Instance, transformType=TRANSFORMPOSSOURCE_World}) {b=true, g=true, r=true}
  FunctionOutput(None=@n26) {outputName="Object  Pivot Point", sortPriority=20}
  FunctionOutput() {description=Empty, outputName=------------------------, sortPriority=9}
  FunctionOutput() {description=Empty, outputName="------------------------ ", sortPriority=16}
  FunctionOutput() {description=Empty, outputName="------------------------  ", sortPriority=30}
  n27 = Frac(None=@n18)
  n28 = Transform(None=Normalize(VectorInput=ConstantBiasScale(None=@n27) {bias=-0.5, scale=2})) {transformSourceType=TRANSFORMSOURCE_Instance}
  FunctionOutput(None=Normalize(VectorInput=@n28)) {
   description=Returns a normalized vector pointing down the branches x-axis. This is not typically needed for foliage animations unless it is needed for custom vector calculations.
   outputName=Leaf X-Axis Vector
   sortPriority=12
  }
  FunctionOutput(None=DotProduct(A=@n26, B=@n20.Normalized Wind Vector)) {description=Gradient aligned with the wind vector at object level in world space., outputName=Object Flow Grad, sortPriority=24}
  n29 = Constant3Vector() {constant=<rgba 0 0 1 0>}
  FunctionOutput(None=DotProduct(A=@n26, B=CrossProduct(A=@n20.Normalized Wind Vector, B=@n29))) {
   description=Gradient aligned to the wind vector at a 90 degree angle at object level in world space.
   outputName=Object Flow Grad 90 deg
   sortPriority=25
  }
  n30 = Normalize(VectorInput=Transform(None=Constant3Vector() {constant=<rgba 0 0 1 0>}) {transformSourceType=TRANSFORMSOURCE_Instance})
  FunctionOutput(None=Normalize(VectorInput=CrossProduct(A=@n30, B=@n20.Normalized Wind Vector))) {outputName=Object Rotation Axis, sortPriority=22}
  FunctionOutput(None=@n20.Normalized Wind Vector) {outputName=Normalized Wind Vector, sortPriority=32}
  FunctionOutput(None=@n30) {outputName=Object Orientation, sortPriority=21}
  FunctionOutput(None=DotProduct(A=@n5, B=@n17)) {description=Returns a gradient value in the direction of the blowing wind., outputName=Branch flow Grad, sortPriority=3}
  FunctionOutput(None=Normalize(VectorInput=CrossProduct(A=@n28, B=@n17))) {
   description="Returns a rotation axis for leaves to be used with the \"RotateAboutAxis\" node. Branches return (0,0,0). Note: The angle is found by transforming a vector along the leaf's x-axis from local to world space. Then the cross product is found between the wind direction and the transformed vector."
   outputName=Leaf Rot Axis
   sortPriority=11
  }
  n31 = TextureCoordinate() {coordinateIndex=3}
  n32 = Constant()
  n33 = Constant() {r=1}
  FunctionOutput(None=If(A=ComponentMask(None=@n31) {desc=Leaf Z height The rest is black, r=true}, B=@n32, "A > B"=@n33, "A == B"=@n32, "A < B"=@n33) {desc=Leaf Mask}) {description=Returns a white mask for leaves. All branches are black., outputName=Leaf Mask, sortPriority=13}
  FunctionOutput(None=@n19) {
   description="Return the pivot position information in world space for each leaf. Branches return a value of (0,0,0)."
   outputName=Leaf Piv Position
   sortPriority=10
  }
  n34 = Constant3Vector() {constant=<rgba 0 0 1 0>}
  FunctionOutput(None=DotProduct(A=@n5, B=CrossProduct(A=@n17, B=@n34))) {
   description=Returns a gradient value across world space at a 90 degree angle to the wind.
   outputName=Branch flow Grad 90 deg
   sortPriority=4
  }
 ]
}
